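# Physics asset for the M_MED_EmailStamp male medium player body cosmetic, covering a right-side glute skirt bone chain.
# Capsule bodies on the pelvis, a small skirt base and a larger, offset dynamic skirt bone form a three-link chain.
# The base joint to the pelvis has free swing and twist, while the dynamic bone's joint to the base has asymmetric swing limits and almost no twist.
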PhysicsAsset /BRCosmetics/Characters/Player/Male/Medium/Bodies/M_MED_EmailStamp/Meshes/M_MED_EmailStamp_Physics
bodies {
 base_glute_skirt_r [{shapeType=Capsule, length=0.505, radius=0.505, shapeName=base_glute_skirt_r_capsule}]
 dyn_glute_skirt_r [{shapeType=Capsule, length=6.1824, radius=4.2437, shapeName=dyn_glute_skirt_r_capsule, center=<4.2233 0 0>}]
 pelvis [{shapeType=Capsule, length=5, radius=5}]
}
constraints [
 {
  bone1Name=base_glute_skirt_r
  bone2Name=pelvis
  swing1LimitDegrees=45
  swing1Motion=Free
  swing2LimitDegrees=45
  swing2Motion=Free
  twistLimitDegrees=45
  twistMotion=Free
 }
 {bone1Name=dyn_glute_skirt_r, bone2Name=base_glute_skirt_r, swing1LimitDegrees=40, swing2LimitDegrees=20, twistLimitDegrees=1}
]
