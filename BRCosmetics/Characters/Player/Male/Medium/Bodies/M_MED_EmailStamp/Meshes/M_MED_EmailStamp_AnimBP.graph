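# Post-process style AnimBP for the M_MED_EmailStamp character body, copying the attached parent's pose and adding secondary motion to skirt, gun and canister.
# Weighted constraints and additive Gun rotations, partly driven by left-thigh pitch and roll against the pelvis, feed pelvis-space RigidBody physics, then a Control Rig.
# The event graph computes those leg alphas, gravity and sim-space settings with a front-end override, plus sitting and backbling flags.
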AnimBlueprint /BRCosmetics/Characters/Player/Male/Medium/Bodies/M_MED_EmailStamp/Meshes/M_MED_EmailStamp_AnimBP
target_skeleton=&/M_MED_EmailStamp_Skeleton
graphs [
 AnimGraph {nodes [
   "Constraint - Gun (PP)"!(ComponentPose, bAlphaBoolEnabled:bool=True, ConstraintWeights_0=0.8, ConstraintWeights_1:float=0.2) -> (Pose) {
    alphaInputType=Bool
    alphaBoolBlend {blendInTime=0.2, blendOption=QuinticInOut, blendOutTime=0.2}
    boneToModify {boneName=Gun}
    constraintSetup [{targetBone {boneName=thigh_l}}, {targetBone {boneName=pelvis}}]
    constraintWeights [1,1]
   }
   n1 = "Constraint - base_glute_skirt_r (PPPPPPP)"!(
    ComponentPose=Local To Component(LocalPose=Copy Pose From Mesh() {bUseAttachedParent=true})
    Alpha=1
    ConstraintWeights_0=0.7844
    ConstraintWeights_1:float=0.0033
    ConstraintWeights_2:float=0.0051
    ConstraintWeights_3:float=0.0014
    ConstraintWeights_4:float=0.066
    ConstraintWeights_5:float=0.1305
    ConstraintWeights_6:float=0.0093
   ) -> (Pose) {
    boneToModify {boneName=base_glute_skirt_r}
    constraintSetup [
     {targetBone {boneName=pelvis}}
     {targetBone {boneName=spine_01}}
     {targetBone {boneName=spine_02}}
     {targetBone {boneName=spine_03}}
     {targetBone {boneName=thigh_twist_01_r}}
     {targetBone {boneName=deform_glute_r}}
     {targetBone {boneName=dyn_glute_skirt_r}}
    ]
    constraintWeights [1,1,1,1,1,1,1]
   }
   n2 = "Constraint - Gun (PPP)"!(ComponentPose=@n1, Alpha=1, ConstraintWeights_0=0.0728, ConstraintWeights_1:float=0.1057, ConstraintWeights_2:float=0.8216) -> (Pose) {
    boneToModify {boneName=Gun}
    constraintSetup [{targetBone {boneName=pelvis}}, {targetBone {boneName=thigh_l}}, {targetBone {boneName=thigh_twist_01_l}}]
    constraintWeights [1,1,1]
   }
   n3 = "Constraint - Canister (PPPPP)"!(
    ComponentPose=@n2
    Alpha=1
    ConstraintWeights_0=0.6404
    ConstraintWeights_1:float=0.1348
    ConstraintWeights_2:float=0.1254
    ConstraintWeights_3:float=0.0909
    ConstraintWeights_4:float=0.0086
   ) -> (Pose) {
    boneToModify {boneName=Canister}
    constraintSetup [
     {targetBone {boneName=pelvis}}
     {targetBone {boneName=spine_01}}
     {targetBone {boneName=spine_02}}
     {targetBone {boneName=spine_03}}
     {targetBone {boneName=spine_04}}
    ]
    constraintWeights [1,1,1,1,1]
   }
   n4 = "Transform (Modify) Bone - Bone: Gun"(ComponentPose=@n3, Alpha=Get Leg rot() -> (Leg rot:double), Translation="0,0,0", Rotation="-7.396,-56.275,111.973", Scale="1,1,1") {
    rotationMode=BMM_Additive
    rotationSpace=BCS_ParentBoneSpace
    alphaScaleBiasClamp {bInterpResult=true}
    boneToModify {boneName=Gun}
    rotation=<rot -7.396 -56.275 111.973>
   }
   n5 = "Transform (Modify) Bone - Bone: Gun"(ComponentPose=@n4, Alpha=Get Leg roll() -> (Leg roll:double), Translation="0,0,0", Rotation="-7.430,-2.530,15.371", Scale="1,1,1") {rotationMode=BMM_Additive, rotationSpace=BCS_ParentBoneSpace, boneToModify {boneName=Gun}, rotation=<rot -7.43 -2.53 15.371>}
   n6 = "Transform (Modify) Bone - Bone: Gun"!(ComponentPose=@n5, bAlphaBoolEnabled:bool=True, Translation="0,0,0", Rotation="15.884,3.784,-5.057", Scale="1,1,1") -> (Pose) {
    alphaInputType=Bool
    rotationMode=BMM_Additive
    rotationSpace=BCS_ParentBoneSpace
    alphaBoolBlend {blendInTime=0.2, blendOption=QuinticInOut, blendOutTime=0.2}
    boneToModify {boneName=Gun}
    rotation=<rot 15.884 3.784 -5.057>
   }
   n7 = RigidBody!(
    ComponentPose=@n6
    bAlphaBoolEnabled:bool
    OverrideWorldGravity:Vector="0,0,0"
    ExternalForce
    SimSpaceSettings:Sim Space Settings Structure=(WorldAlpha=0,VelocityScaleZ=1,DampingAlpha=1,MaxLinearVelocity=10000,MaxAngularVelocity=10000,MaxLinearAcceleration=10000,MaxAngularAcceleration=10000,ExternalLinearDragV=(X=0,Y=0,Z=0),ExternalLinearVelocity=(X=0,Y=0,Z=0),ExternalAngularVelocity=(X=0,Y=0,Z=0))
   ) -> (Pose) {
    alphaInputType=Bool
    bOverrideWorldGravity=true
    overridePhysicsAsset=&/M_MED_EmailStamp_Physics
    simulationSpace=BaseBoneSpace
    baseBoneRef {boneName=pelvis}
    componentLinearAccScale {x=0.2}
    componentLinearVelScale {x=0.2}
   }
   n8 = Control Rig(Source=Component To Local(ComponentPose=@n7)) {controlRigAssetReference {controlRigAsset=&/M_MED_EmailStamp_CtrlRig}}
   Output Pose(Result=@n8) {layerGroup=None}
  ]}
 EventGraph {nodes [
   Event BlueprintUpdateAnimation()
   > n1 = Sequence()
   > n2 = "Gravity Override State Function (Parameter Asset) (Nativized)"(
    CharacterPartAnimInstance=Self-Reference()
    GameTimeInSeconds=0
    Parameters=&/M_MED_EmailStamp_Parameters
   )
   > Set Gravity Override(Gravity Override=@n2)
   > n3 = Branch(Condition=Get bIsInFrontEnd())
   > Set Sim Space Settings(Sim Space Settings:Sim Space Settings Structure=Make Sim Space Settings!(
     WorldAlpha:float=0.1
     VelocityScaleZ:float=1
     DampingAlpha:float=1
     MaxLinearVelocity:float=800
     MaxAngularVelocity:float=8
     MaxLinearAcceleration:float=1000
     MaxAngularAcceleration:float=1000
     ExternalLinearDragV:Vector="0,0,0"
     ExternalLinearVelocity:Vector="0,0,0"
     ExternalAngularVelocity:Vector="0,0,0"
    ) -> (SimSpaceSettings:Sim Space Settings Structure) {is=title K2Node_MakeStruct}) -> (Output_Get:Sim Space Settings Structure)
   > Set Gravity Override(Gravity Override="0,0,-500")
   Set Frontend Alpha(execute=@n3.else, Frontend Alpha:double=0) -> (Output_Get:double)
   > Set Sim Space Settings()
   n4 = GetDeltaTransformFromRefPose!(self=GetOwnerSkeletalMeshComponent(), BoneName=thigh_l, BaseName=pelvis) -> (
    ReturnValue_Location:Vector
    ReturnValue_Rotation_Roll:float
    ReturnValue_Rotation_Pitch:float
    ReturnValue_Rotation_Yaw:float
    ReturnValue_Scale:Vector
   )
   Set Leg roll(execute=@n1.then_1, Leg roll:double=MapRangeClamped(Value=@n4.ReturnValue_Rotation_Roll, InRangeA=20, InRangeB=-40)) -> (Output_Get:double)
   > Set Leg rot(Leg rot:double=MapRangeClamped(Value=@n4.ReturnValue_Rotation_Pitch, InRangeA=10, InRangeB=80)) -> (Output_Get:double)
   n5 = Get bIsMoving2D()
   > Set Is sitting(Is sitting:bool=AND Boolean(A=NOT Boolean(A=@n5), B=Get bIsCrouching())) -> (Output_Get:bool)
   Event OnCharacterCustomizationLoaded()
   n6 = CheckForBackbling()
   > Set is Wearing Backbling(is Wearing Backbling:bool=OR Boolean!(A=@n6, B=@n6.bIsBackblingACape, C:bool=@n6.bIsBackblingLong) -> (ReturnValue)) -> (Output_Get:bool)
   Break Fort Rigid Body Anim Node Parameters!(FortRigidBodyAnimNodeParameters:"Fort Rigid Body Anim Node Parameters Structure (by ref)"=@n2.OutRigidBodyAnimNodeParameters) -> (ComponentLinearAccScale:Vector=~, ComponentLinearVelScale:Vector=~, SimSpaceSettings:Sim Space Settings Structure) {is=title K2Node_BreakStruct}
  ]}
]
variables [
 Frontend Alpha {type=number}
 Gravity Override {type=Vector}
 Is sitting {type=boolean}
 is Wearing Backbling {type=boolean}
 Leg roll {type=number}
 Leg rot {type=number}
 Sim Space Settings {
  type=SimSpaceSettings
  default {
   dampingAlpha=1
   maxAngularAcceleration=10000
   maxAngularVelocity=10000
   maxLinearAcceleration=10000
   maxLinearVelocity=10000
   velocityScaleZ=1
   worldAlpha=0
   externalLinearDragV=<0 0 0>
  }
 }
]
