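# Physics asset for the face accessory part of the EmailStamp medium male player cosmetic, covering its back scarf and hood bones.
# Capsule bodies form a two-link back scarf chain and single-link back and right-shoulder hood chains, each hung from a base bone constrained to spine_05.
# Base-to-spine and spine joints swing and twist freely, while the dyn_ joints have tight swing limits and little or no twist.
# Large spine_04 and spine_05 capsules represent the upper torso.
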PhysicsAsset /BRCosmetics/Characters/Player/Male/Medium/Bodies/M_MED_EmailStamp/Meshes/Parts/M_MED_EmailStamp_FaceAcc_Physics
bodies {
 base_back_hood [{shapeType=Capsule, length=2, radius=2, shapeName=base_back_hood_capsule, rotation=<rot 0 0 90>}]
 base_scarf_bk [{shapeType=Capsule, length=0.505, radius=0.505, shapeName=base_scarf_bk_capsule, rotation=<rot 0 -90 0>}]
 base_shoulder_hood_r [{shapeType=Capsule, length=2, radius=2, shapeName=base_shoulder_hood_r_capsule, rotation=<rot 0 0 90>}]
 dyn_back_hood [{shapeType=Capsule, length=2, radius=2, shapeName=dyn_back_hood_capsule, center=<2.7106 0 0>, rotation=<rot 0 0 90>}]
 dyn_scarf_bk_01 [{shapeType=Capsule, length=10.2144, radius=6.5033, shapeName=dyn_scarf_bk_01_capsule, center=<9.5 -6.4916 0>, rotation=<rot 0 -90 0>}]
 dyn_scarf_bk_02 [{shapeType=Capsule, length=7.9182, radius=6.8569, shapeName=dyn_scarf_bk_02_capsule, center=<9.5 -6.1201 0>, rotation=<rot 0 -90 0>}]
 dyn_shoulder_hood_r [{shapeType=Capsule, length=2, radius=2, shapeName=dyn_shoulder_hood_r_capsule, center=<2.7106 0 0>, rotation=<rot 0 0 90>}]
 spine_04 [{shapeType=Capsule, length=19.5, radius=12.8225, shapeName=spine_04_capsule, center=<0 -0.8053 0>}]
 spine_05 [{shapeType=Capsule, length=20, radius=12, center=<0 -0.4188 0>}]
}
constraints [
 {
  bone1Name=base_scarf_bk
  bone2Name=spine_05
  swing1LimitDegrees=45
  swing1Motion=Free
  swing2LimitDegrees=45
  swing2Motion=Free
  twistLimitDegrees=45
  twistMotion=Free
 }
 {bone1Name=dyn_scarf_bk_01, bone2Name=base_scarf_bk, swing1LimitDegrees=30, swing2LimitDegrees=10, twistLimitDegrees=0}
 {bone1Name=dyn_scarf_bk_02, bone2Name=dyn_scarf_bk_01, swing1LimitDegrees=30, swing2LimitDegrees=30, twistLimitDegrees=0}
 {
  bone1Name=spine_05
  bone2Name=spine_04
  swing1LimitDegrees=45
  swing1Motion=Free
  swing2LimitDegrees=45
  swing2Motion=Free
  twistLimitDegrees=45
  twistMotion=Free
 }
 {bone1Name=dyn_shoulder_hood_r, bone2Name=base_shoulder_hood_r, swing1LimitDegrees=20, swing2LimitDegrees=40, twistLimitDegrees=1}
 {
  bone1Name=base_shoulder_hood_r
  bone2Name=spine_05
  swing1LimitDegrees=45
  swing1Motion=Free
  swing2LimitDegrees=45
  swing2Motion=Free
  twistLimitDegrees=45
  twistMotion=Free
 }
 {bone1Name=dyn_back_hood, bone2Name=base_back_hood, swing1LimitDegrees=20, swing2LimitDegrees=40, twistLimitDegrees=1}
 {
  bone1Name=base_back_hood
  bone2Name=spine_05
  swing1LimitDegrees=45
  swing1Motion=Free
  swing2LimitDegrees=45
  swing2Motion=Free
  twistLimitDegrees=45
  twistMotion=Free
 }
]
