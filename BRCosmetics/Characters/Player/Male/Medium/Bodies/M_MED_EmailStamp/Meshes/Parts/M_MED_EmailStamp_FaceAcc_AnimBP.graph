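# Anim Blueprint for the EmailStamp male character's face accessory, animating a hood with secondary motion.
# Poses copied from the attached parent and character mesh are layered at the upper arms, then run through Control Rig and weighted shoulder/back hood constraints.
# AnimDynamics on three hood bones feeds a spine_04-relative RigidBody simulation.
# Event graph sets gravity from a parameter asset, front-end simulation settings, backbling flags and the source mesh.
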AnimBlueprint /BRCosmetics/Characters/Player/Male/Medium/Bodies/M_MED_EmailStamp/Meshes/Parts/M_MED_EmailStamp_FaceAcc_AnimBP
target_skeleton=&/M_MED_EmailStamp_FaceAcc_Skeleton
graphs [
 AnimGraph {nodes [
   n1 = Get CharacterMesh() -> (CharacterMesh:Skeletal Mesh Component Object Reference)
   n2 = Control Rig(Source=Layered blend per bone(
     BasePose=Copy Pose From Mesh() {bUseAttachedParent=true}
     BlendPoses_0=Copy Pose From Mesh(SourceMeshComponent=@n1) {bUseAttachedParent=true}
    ) {layerSetup [{branchFilters [{boneName=upperarm_l}, {boneName=upperarm_r}]}]}) {controlRigAssetReference {controlRigAsset=&/M_MED_EmailStamp_FaceAcc_CtrlRig}}
   n3 = "Constraint - base_back_hood (PPPPPPP)"!(
    ComponentPose=Local To Component(LocalPose=@n2)
    Alpha=1
    ConstraintWeights_0=0.0512
    ConstraintWeights_1:float=0.0108
    ConstraintWeights_2:float=0.1078
    ConstraintWeights_3:float=0.2536
    ConstraintWeights_4:float=0.2397
    ConstraintWeights_5:float=0.1051
    ConstraintWeights_6:float=0.2318
   ) -> (Pose) {
    boneToModify {boneName=base_back_hood}
    constraintSetup [
     {targetBone {boneName=spine_03}}
     {targetBone {boneName=spine_04}}
     {targetBone {boneName=spine_05}}
     {targetBone {boneName=clavicle_r}}
     {targetBone {boneName=upperarm_twist_01_r}}
     {targetBone {boneName=deform_upperarm_bk_base_r}}
     {targetBone {boneName=deform_upperarm_up_r}}
    ]
    constraintWeights [1,1,1,1,1,1,1]
   }
   n4 = "Constraint - base_shoulder_hood_r (PPPPPPP)"!(
    ComponentPose=@n3
    Alpha=1
    ConstraintWeights_0=0.014
    ConstraintWeights_1:float=0.1586
    ConstraintWeights_2:float=0.2581
    ConstraintWeights_3:float=0.009
    ConstraintWeights_4:float=0.3414
    ConstraintWeights_5:float=0.0318
    ConstraintWeights_6:float=0.187
   ) -> (Pose) {
    boneToModify {boneName=base_shoulder_hood_r}
    constraintSetup [
     {targetBone {boneName=spine_03}}
     {targetBone {boneName=spine_05}}
     {targetBone {boneName=clavicle_r}}
     {targetBone {boneName=upperarm_r}}
     {targetBone {boneName=upperarm_twist_01_r}}
     {targetBone {boneName=deform_upperarm_fr_r}}
     {targetBone {boneName=deform_upperarm_up_r}}
    ]
    constraintWeights [1,1,1,1,1,1,1]
   }
   n5 = Anim Dynamics - Bone: dyn_hoodup_l!(SimulationSpace:AnimPhysSimSpaceType Enum, ComponentPose=@n4, Alpha=1, ExternalForce="0,0,0") -> (Pose) {
    angularDampingOverride=0.9999
    angularSpringConstant=300
    bLinearSpring=true
    bOverrideAngularDamping=true
    bOverrideLinearDamping=true
    bUseGravityOverride=true
    linearDampingOverride=0.9999
    linearSpringConstant=50
    simulationSpace=RootRelative
    alphaBoolBlend {blendOption=Linear}
    boundBone {boneName=dyn_hoodup_l}
    componentAppliedLinearAccClamp=<10000 10000 10000>
    componentLinearAccScale=<0.3 0.3 0.3>
    componentLinearVelScale=<0.4 0.4 0.4>
    gravityOverride {z=-980}
    physicsBodyDefinitions [{
      boundBone {boneName=dyn_hoodup_l}
      boxExtents=<2 2 2>
      constraintSetup {angularLimitsMax=<0 0 0>, angularLimitsMin=<0 0 0>, angularTarget=<1 0 0>, linearAxesMax=<6 4 2>, linearAxesMin=<-6 -4 -2>}
      localJointOffset=<2 0 0>
     }]
   }
   n6 = Anim Dynamics - Bone: dyn_hoodup_r!(SimulationSpace:AnimPhysSimSpaceType Enum, ComponentPose=@n5, Alpha=1, ExternalForce="0,0,0") -> (Pose) {
    angularDampingOverride=0.9999
    angularSpringConstant=300
    bLinearSpring=true
    bOverrideAngularDamping=true
    bOverrideLinearDamping=true
    bUseGravityOverride=true
    linearDampingOverride=0.9999
    linearSpringConstant=50
    simulationSpace=RootRelative
    alphaBoolBlend {blendOption=Linear}
    boundBone {boneName=dyn_hoodup_r}
    componentAppliedLinearAccClamp=<10000 10000 10000>
    componentLinearAccScale=<0.3 0.3 0.3>
    componentLinearVelScale=<0.4 0.4 0.4>
    gravityOverride {z=-980}
    physicsBodyDefinitions [{
      boundBone {boneName=dyn_hoodup_r}
      boxExtents=<2 2 2>
      constraintSetup {angularLimitsMax=<0 0 0>, angularLimitsMin=<0 0 0>, angularTarget=<1 0 0>, linearAxesMax=<6 4 1>, linearAxesMin=<-6 -4 -1>}
      localJointOffset=<2 0 0>
     }]
   }
   n7 = Anim Dynamics - Bone: dyn_hoodup_bk!(SimulationSpace:AnimPhysSimSpaceType Enum, ComponentPose=@n6, Alpha=1, ExternalForce="0,0,0") -> (Pose) {
    angularDampingOverride=0.9999
    angularSpringConstant=50
    bAngularSpring=true
    bLinearSpring=true
    bOverrideAngularDamping=true
    bOverrideLinearDamping=true
    bUseGravityOverride=true
    linearDampingOverride=0.9999
    linearSpringConstant=50
    simulationSpace=RootRelative
    alphaBoolBlend {blendOption=Linear}
    boundBone {boneName=dyn_hoodup_bk}
    componentAppliedLinearAccClamp=<10000 10000 10000>
    componentLinearAccScale=<0.3 0.3 0.3>
    componentLinearVelScale=<0.4 0.4 0.4>
    gravityOverride {z=-980}
    physicsBodyDefinitions [{
      boundBone {boneName=dyn_hoodup_bk}
      boxExtents=<2 2 2>
      constraintSetup {angularLimitsMax=<2 2 2>, angularLimitsMin=<-2 -2 -2>, angularTarget=<1 0 0>, linearAxesMax=<6 6 4>, linearAxesMin=<-6 -6 0>}
      localJointOffset=<1 0 0>
     }]
   }
   n8 = RigidBody!(
    ComponentPose=@n7
    bAlphaBoolEnabled:bool
    OverrideWorldGravity:Vector="0,0,0"
    ExternalForce
    SimSpaceSettings:Sim Space Settings Structure=(WorldAlpha=0,VelocityScaleZ=1,DampingAlpha=1,MaxLinearVelocity=10000,MaxAngularVelocity=10000,MaxLinearAcceleration=10000,MaxAngularAcceleration=10000,ExternalLinearDragV=(X=0,Y=0,Z=0),ExternalLinearVelocity=(X=0,Y=0,Z=0),ExternalAngularVelocity=(X=0,Y=0,Z=0))
   ) -> (Pose) {
    alphaInputType=Bool
    bOverrideWorldGravity=true
    overridePhysicsAsset=&/M_MED_EmailStamp_FaceAcc_Physics
    simulationSpace=BaseBoneSpace
    baseBoneRef {boneName=spine_04}
    componentLinearAccScale=<0.6 0.12 0.12>
    componentLinearVelScale=<0.6 0.12 0.12>
   }
   Output Pose(Result=Component To Local(ComponentPose=@n8)) {layerGroup=None}
  ]}
 EventGraph {nodes [
   n1 = "Gravity Override State Function (Parameter Asset) (Nativized)"(
    CharacterPartAnimInstance=Self-Reference()
    GameTimeInSeconds=GetGameTimeInSeconds()
    Parameters=/BRCosmetics/Characters/Player/Female/Medium/Bodies/F_MED_MasalaRail/Meshes/Parts/F_MED_MasalaRail_FaceAcc_Parameters
   )
   > Set Gravity Override(Gravity Override=@n1)
   n2 = Event BlueprintUpdateAnimation()
   > n3 = Sequence()
   > n4 = Branch(Condition=Get bIsInFrontEnd())
   > Set Simulation Space(Simulation Space:AnimPhysSimSpaceType Enum=World) -> (Output_Get:AnimPhysSimSpaceType Enum)
   > Set AnimDynAlpha(AnimDynAlpha:double=1) -> (Output_Get:double)
   Set Simulation Space(execute=@n4.else)
   > Set AnimDynAlpha(AnimDynAlpha=1)
   n5 = "Gravity Override State Function (Parameter Asset) (Nativized)"(
    execute=@n3.then_1
    CharacterPartAnimInstance=Self-Reference()
    GameTimeInSeconds=@n2.DeltaTimeX
    Parameters=&/M_MED_EmailStamp_FaceAcc_Parameters
   )
   > Set Gravity Override(Gravity Override=@n5)
   > n6 = Branch(Condition=Get bIsInFrontEnd())
   > Set Sim Space Settings(Sim Space Settings:Sim Space Settings Structure=Make Sim Space Settings!(
     WorldAlpha:float=0.1
     VelocityScaleZ:float=1
     DampingAlpha:float=1
     MaxLinearVelocity:float=800
     MaxAngularVelocity:float=8
     MaxLinearAcceleration:float=1000
     MaxAngularAcceleration:float=1000
     ExternalLinearDragV:Vector="0,0,0"
     ExternalLinearVelocity:Vector="0,0,0"
     ExternalAngularVelocity:Vector="0,0,0"
    ) -> (SimSpaceSettings:Sim Space Settings Structure) {is=title K2Node_MakeStruct}) -> (Output_Get:Sim Space Settings Structure)
   > Set Gravity Override(Gravity Override="0,0,-500")
   Set Frontend Alpha(execute=@n6.else, Frontend Alpha:double=0) -> (Output_Get:double)
   > Set Sim Space Settings()
   n7 = Event OnCharacterCustomizationLoaded()
   n8 = CheckForBackbling()
   n9 = OR Boolean!(A=@n8, B=@n8.bIsBackblingACape, C:bool=@n8.bIsBackblingLong) -> (ReturnValue)
   > Set isWearingPack(isWearingPack:bool=@n9) -> (Output_Get:bool)
   > Set isWearingCape(isWearingCape=NOT Boolean(A=@n9))
   > Set CharacterMesh(CharacterMesh:Skeletal Mesh Component Object Reference=@n7.CustomBaseMesh) -> (Output_Get:Skeletal Mesh Component Object Reference)
   Break Fort Rigid Body Anim Node Parameters!(FortRigidBodyAnimNodeParameters:"Fort Rigid Body Anim Node Parameters Structure (by ref)"=@n5.OutRigidBodyAnimNodeParameters) -> (ComponentLinearAccScale:Vector=~, ComponentLinearVelScale:Vector=~, SimSpaceSettings:Sim Space Settings Structure) {is=title K2Node_BreakStruct}
   Reroute Node(InputPin:bool=@n8.bIsBackblingLong) -> (OutputPin:bool)
  ]}
]
variables [
 AnimDynAlpha {type=number}
 CharacterMesh {type=SkeletalMeshComponent}
 Frontend Alpha {type=number}
 Gravity Override {type=Vector}
 is Wearing Backbling {type=boolean}
 isWearingCape {type=boolean}
 isWearingPack {type=boolean}
 Sim Space Settings {
  type=SimSpaceSettings
  default {
   dampingAlpha=1
   maxAngularAcceleration=10000
   maxAngularVelocity=10000
   maxLinearAcceleration=10000
   maxLinearVelocity=10000
   velocityScaleZ=1
   worldAlpha=0
   externalLinearDragV=<0 0 0>
  }
 }
 Simulation Space {type=AnimPhysSimSpaceType}
]
